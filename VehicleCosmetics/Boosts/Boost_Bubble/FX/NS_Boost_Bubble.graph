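# Niagara vehicle boost cosmetic that streams bubbles backward from skeletal mesh sockets and pops them with splashes.
# Three sprite emitters layer large and small bubbles with vortex, velocity-damping and inherited-movement motion; the larger two also collide.
# Bubbles write to a death array, and system-level death variables trigger animated SubUV splashes, avoiding particle events.
# GameModeID switches spawn rate and lifetime, though two emitters hardcode the minimum lifetime's index to 0.
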NiagaraSystem /VehicleCosmetics/Boosts/Boost_Bubble/FX/NS_Boost_Bubble
dependencies {
 usedDataInterfaces [
  &NiagaraDataInterfaceCurve
  &NiagaraDataInterfaceSpriteRendererInfo
  &NiagaraDataInterfaceSkeletalMesh
  &NiagaraDataInterfaceArrayPosition
  &NiagaraDataInterfaceArrayFloat3
 ]
 usedDynamicInputs [&SelectVectorFromArray]
 usedModules [
  &self:SystemSpawnScript.SystemScriptSource.SystemScriptGraph.NiagaraNodeAssignment_0.SetVariables_#1
  &self:SystemSpawnScript.SystemScriptSource.SystemScriptGraph.NiagaraNodeAssignment_2.SetVariables_#2
  &SystemState
  &self:ReadDeath
  &self:SystemSpawnScript.SystemScriptSource.SystemScriptGraph.NiagaraNodeAssignment_1.SetVariables_#3
  &self:SystemSpawnScript.SystemScriptSource.SystemScriptGraph.NiagaraNodeAssignment_3.SetVariables_#4
 ]
}
emitters {
 NE_VC_Bubbles {
  emitterUpdateScript [
   &EmitterState(Life Cycle Mode=::NewEnumerator1, Loop Duration=2, UseLoopDelay=false)
   &VC_FilteredSocketCount()
   &SpawnRate(
    SpawnRate=&MergeHelper_FloatSwitch(0=40, 1=40, 2=40, 3=120, Index=$User.GameModeID, Num=::NewEnumerator2)
    Use Spawn Probability=false
   )
  ]
  particleSpawnScript [
   &VC_PaintedVariantsOverride() {enabled=false}
   &InitializeParticle(
    Color=<rgba 4 4 4 1>
    Lifetime Max=&MergeHelper_FloatSwitch(0=2, 1=2, 2=2, 3=0.1, Index=$User.GameModeID, Num=::NewEnumerator2)
    Lifetime Min:NiagaraFloat=&MergeHelper_FloatSwitch(0=1, 1=1, 2=1, 3=0.08, Index=$User.GameModeID, Num=::NewEnumerator2)
    Lifetime Mode=::NewEnumerator1
    Mass Mode=::NewEnumerator3
    Sprite Size Mode=::NewEnumerator3
    Uniform Sprite Size=13
    UsePositionOffset=false
    Write Material Random=true
   )
   &PartitionParticles(Number Of Partitions=$Emitter.PortCount)
   &SkeletalMeshLocation(
    Bone / Socket Index=$Particles.PartitionParticles.Partition
    Bone / Socket Sampling Mode=::NewEnumerator5
    MaskSpawnGroup=false
    Mesh Sampling Type=::NewEnumerator6
    Sampled Position Offset=&AddVectorToPosition(Position=<0 0 0>, Vector=Linked() {linkedVariable {name=User.DTCarOffset, type=NiagaraPosition}})
    Skeletal Mesh=$System.SkeletalMesh
   )
   &AddVelocity(Velocity=<-750 0 0>)
   &AddVelocity(Velocity=&RandomRangeVector(Minimum=<-65 -100 -16>, Maximum=<64 100 128>)) {module=AddVelocity001}
   &AccelerationForce(Acceleration=&RandomRangeVector(Minimum=<0 0 0>, Maximum=<0 0 50>))
   &ApplyInitialForces()
   &ShapeLocation(MaskSpawnGroup=false, Override Seed=false, Sphere Radius=40, UseOwnerScale=true)
   /VehicleCosmetics/FX/VC_FXParents/Modules/DLR_ProductPreviews(GameModeID:NiagaraInt32=$System.GameModeID, bPreviewVehicleSpeed:NiagaraBool=$System.bPreviewVehicleSpeed)
  ]
  particleUpdateScript [
   &ParticleState()
   &self:WriteDeath_02(Position Array:NiagaraDataInterfaceArrayPosition=$System.DeathArray) {module=WriteDeath_02}
   &ScaleSpriteSize(Uniform Curve Sprite Scale=DataInterface() {propertyValues {Curve {keys [
        <at 0 0.2 RCIM_Cubic>
        <at 0.1 2>
        {arriveTangent=2.2222, interpMode=RCIM_Cubic, leaveTangent=2.2222, time=0.8, value=2}
        <at 1 4 RCIM_Cubic>
       ]}}})
   &ScaleVelocity(
    Coordinate Space:ENiagaraCoordinateSpace=::Simulation
    Velocity Scale=&RandomRangeVector(
     Minimum=&VectorFromFloat(Value=&FloatFromCurve(FloatCurve=DataInterface() {propertyValues {Curve {keys [
           <at 0.01 1 RCIM_Cubic>
           {arriveTangent=0.69, interpMode=RCIM_Cubic, leaveTangent=0.69, tangentMode=RCTM_User, time=0.35, value=0.3}
           <at 0.5 0.2 RCIM_Cubic>
          ]}}}))
     Maximum=&VectorFromFloat(Value=&FloatFromCurve(FloatCurve=DataInterface() {propertyValues {Curve {keys [
           <at 0.01 1 RCIM_Cubic>
           {arriveTangent=0.69, interpMode=RCIM_Cubic, leaveTangent=0.69, tangentMode=RCTM_User, time=0.4, value=0.35}
           <at 0.5 0.45 RCIM_Cubic>
          ]}}}))
    )
   )
   &VortexForce(
    UseOriginOffset=false
    Vortex Axis=&RandomRangeVector(Minimum=<0 0 0>, Maximum=<0 16 0>)
    Vortex Force Amount=300
   )
   &Collision(
    Age Colliding Particles=false
    KillOnCollision=false
    Module.Control Rest State Via Surface Normal=false
    Module.EnableMaxCollisionCount=false
    Module.Randomize Collision Normal Vector=false
    Restitution=&RandomRangeFloat(Minimum=0, Maximum=0)
   )
   &InheritSourceMovement(
    Applied Position Scale=&Multiply_VectorByFloat(
     Vector=<1 1 0>
     Float=&FloatFromCurve(FloatCurve=DataInterface() {propertyValues {Curve {keys [
          {
           arriveTangentWeight=0.4252
           interpMode=RCIM_Cubic
           leaveTangentWeight=0.4252
           tangentMode=RCTM_User
           tangentWeightMode=RCTWM_WeightedBoth
           time=0.05
           value=1
          }
          {interpMode=RCIM_Cubic, tangentMode=RCTM_User, time=1, value=0}
         ]}}})
    )
    ApplyForce=false
    ApplyPosition=true
    ApplyPositionFalloff=false
   ) {enabled=false}
   &InheritSourceMovement(
    Applied Position Scale=&Multiply_VectorByFloat(
     Vector=<1 1 1>
     Float=&FloatFromCurve(FloatCurve=DataInterface() {propertyValues {Curve {
         defaultValue=1
         keys [
          {arriveTangent=-1.7885, interpMode=RCIM_Cubic, leaveTangent=-1.7885, tangentMode=RCTM_User, time=0.05, value=1}
          {interpMode=RCIM_Cubic, tangentMode=RCTM_User, time=1, value=0}
         ]
        }}})
    )
    ApplyForce=false
    ApplyPosition=true
    ApplyPositionFalloff=false
   ) {module=InheritSourceMovement001}
   &JitterPosition(Jitter Amount=1.5)
   &SolveForcesAndVelocity(Clamp Velocity=false, Limit Acceleration=false)
   &GenerateDeathEvent(Use Event Probability=false) {enabled=false}
  ]
  rendererClasses [&NiagaraSpriteRendererProperties]
  renderers [{rendererClass=&NiagaraSpriteRendererProperties}]
 }
 NE_VC_Bubbles001 {
  emitterUpdateScript [
   &EmitterState(Life Cycle Mode=::NewEnumerator1, Loop Duration=2, UseLoopDelay=false)
   &VC_FilteredSocketCount()
   &SpawnRate(
    SpawnRate=&MergeHelper_FloatSwitch(0=40, 1=40, 2=40, 3=120, Index=$User.GameModeID, Num=::NewEnumerator2)
    Use Spawn Probability=false
   )
  ]
  particleSpawnScript [
   &VC_PaintedVariantsOverride() {enabled=false}
   &InitializeParticle(
    Color=<rgba 4 4 4 1>
    Lifetime Max=&MergeHelper_FloatSwitch(0=2, 1=2, 2=2, 3=0.1, Index=$User.GameModeID, Num=::NewEnumerator2)
    Lifetime Min:NiagaraFloat=&MergeHelper_FloatSwitch(0=1.5, 1=1.5, 2=1.5, 3=0.05, Index=0, Num=::NewEnumerator2)
    Lifetime Mode=::NewEnumerator1
    Mass Mode=::NewEnumerator3
    Sprite Size Mode=::NewEnumerator3
    Uniform Sprite Size=13
    UsePositionOffset=false
    Write Material Random=true
   )
   &PartitionParticles(Number Of Partitions=$Emitter.PortCount)
   &SkeletalMeshLocation(
    Bone / Socket Index=$Particles.PartitionParticles.Partition
    Bone / Socket Sampling Mode=::NewEnumerator5
    MaskSpawnGroup=false
    Mesh Sampling Type=::NewEnumerator6
    Sampled Position Offset=&AddVectorToPosition(Position=<0 0 0>, Vector=Linked() {linkedVariable {name=User.DTCarOffset, type=NiagaraPosition}})
    Skeletal Mesh=$System.SkeletalMesh
   )
   &AddVelocity(Velocity=<-750 0 0>)
   &AddVelocity(Velocity=&RandomRangeVector(Minimum=<-500 -100 -16>, Maximum=<64 100 128>)) {module=AddVelocity001}
   &AccelerationForce(Acceleration=&RandomRangeVector(Minimum=<0 0 0>, Maximum=<0 0 50>)) {enabled=false}
   &ApplyInitialForces() {enabled=false}
   &ShapeLocation(MaskSpawnGroup=false, Override Seed=false, Sphere Radius=40, UseOwnerScale=true)
   &DLR_ProductPreviews(GameModeID:NiagaraInt32=$System.GameModeID, bPreviewVehicleSpeed:NiagaraBool=$System.bPreviewVehicleSpeed)
  ]
  particleUpdateScript [
   &ParticleState()
   &self:WriteDeath_02(Position Array:NiagaraDataInterfaceArrayPosition=$System.DeathArray) {module=WriteDeath_02}
   &ScaleSpriteSize(Uniform Curve Sprite Scale=DataInterface() {propertyValues {Curve {keys [
        {tangentMode=RCTM_User, time=0, value=1.5}
        {arriveTangent=-0.0934, interpMode=RCIM_Cubic, leaveTangent=-0.0934, tangentMode=RCTM_User, time=0.8, value=1.5}
        <at 1 2.5 RCIM_Cubic>
       ]}}})
   &ScaleVelocity(
    Coordinate Space:ENiagaraCoordinateSpace=::Simulation
    Velocity Scale=&VectorFromFloat(Value=&RandomRangeFloat(
      Minimum=&FloatFromCurve(FloatCurve=DataInterface() {propertyValues {Curve {keys [
           <at 0.01 1 RCIM_Cubic>
           {arriveTangent=0.69, interpMode=RCIM_Cubic, leaveTangent=0.69, tangentMode=RCTM_User, time=0.35, value=0.02}
           <at 0.5 0.2 RCIM_Cubic>
          ]}}})
      Maximum=&FloatFromCurve(FloatCurve=DataInterface() {propertyValues {Curve {keys [
           <at 0.01 1 RCIM_Cubic>
           {arriveTangent=0.69, interpMode=RCIM_Cubic, leaveTangent=0.69, tangentMode=RCTM_User, time=0.35, value=0.15}
           <at 0.5 0.25 RCIM_Cubic>
          ]}}})
     ))
   )
   &VortexForce(
    UseOriginOffset=false
    Vortex Axis=&RandomRangeVector(Minimum=<0 0 0>, Maximum=<0 16 0>)
    Vortex Force Amount=300
   )
   &Collision(
    Age Colliding Particles=false
    KillOnCollision=false
    Module.Control Rest State Via Surface Normal=false
    Module.EnableMaxCollisionCount=false
    Module.Randomize Collision Normal Vector=false
    Restitution=&RandomRangeFloat(Minimum=0, Maximum=0)
   )
   &InheritSourceMovement(
    Applied Position Scale=&Multiply_VectorByFloat(
     Vector=<1 1 0>
     Float=&FloatFromCurve(FloatCurve=DataInterface() {propertyValues {Curve {keys [
          {
           arriveTangentWeight=0.4252
           interpMode=RCIM_Cubic
           leaveTangentWeight=0.4252
           tangentMode=RCTM_User
           tangentWeightMode=RCTWM_WeightedBoth
           time=0.05
           value=1
          }
          {interpMode=RCIM_Cubic, tangentMode=RCTM_User, time=1, value=0}
         ]}}})
    )
    ApplyForce=false
    ApplyPosition=true
    ApplyPositionFalloff=false
   ) {enabled=false}
   &CurlNoiseForce(Bias Noise=false, Noise Strength=50, Pan Noise=false)
   &InheritSourceMovement(
    Applied Position Scale=&Multiply_VectorByFloat(
     Vector=<1 1 1>
     Float=&FloatFromCurve(FloatCurve=DataInterface() {propertyValues {Curve {
         defaultValue=1
         keys [
          {arriveTangent=-1.7885, interpMode=RCIM_Cubic, leaveTangent=-1.7885, tangentMode=RCTM_User, time=0.05, value=1}
          {interpMode=RCIM_Cubic, tangentMode=RCTM_User, time=1, value=0}
         ]
        }}})
    )
    ApplyForce=false
    ApplyPosition=true
    ApplyPositionFalloff=false
   ) {module=InheritSourceMovement001}
   &JitterPosition(Jitter Amount=2) {enabled=false}
   &SolveForcesAndVelocity(Clamp Velocity=false, Limit Acceleration=false)
  ]
  rendererClasses [&NiagaraSpriteRendererProperties]
  renderers [{rendererClass=&NiagaraSpriteRendererProperties}]
 }
 NE_VC_SmallBubbles {
  emitterUpdateScript [
   &EmitterState(Life Cycle Mode=::NewEnumerator1, Loop Duration=2, UseLoopDelay=false)
   &VC_FilteredSocketCount()
   &SpawnRate(
    SpawnRate=&MergeHelper_FloatSwitch(0=40, 1=40, 2=40, 3=120, Index=$User.GameModeID, Num=::NewEnumerator2)
    Use Spawn Probability=false
   )
  ]
  particleSpawnScript [
   &VC_PaintedVariantsOverride() {enabled=false}
   &InitializeParticle(
    Color=<rgba 4 4 4 1>
    Lifetime Max=&MergeHelper_FloatSwitch(0=1, 1=1, 2=1, 3=0.01, Index=$User.GameModeID, Num=::NewEnumerator2)
    Lifetime Min:NiagaraFloat=&MergeHelper_FloatSwitch(0=1, 1=1, 2=1, 3=0.0055, Index=0, Num=::NewEnumerator2)
    Lifetime Mode=::NewEnumerator1
    Mass Mode=::NewEnumerator3
    Sprite Size Mode=::NewEnumerator3
    UsePositionOffset=false
    Write Material Random=true
   )
   &PartitionParticles(Number Of Partitions=$Emitter.PortCount)
   &SkeletalMeshLocation(
    Bone / Socket Index=$Particles.PartitionParticles.Partition
    Bone / Socket Sampling Mode=::NewEnumerator5
    MaskSpawnGroup=false
    Mesh Sampling Type=::NewEnumerator6
    Sampled Position Offset=&AddVectorToPosition(Position=<0 0 0>, Vector=Linked() {linkedVariable {name=User.DTCarOffset, type=NiagaraPosition}})
    Skeletal Mesh=$System.SkeletalMesh
   )
   &AddVelocity(Velocity=<-750 0 0>)
   &AddVelocity(Velocity=&RandomRangeVector(Minimum=<-65 -100 -16>, Maximum=<64 100 128>)) {module=AddVelocity001}
   &AccelerationForce(Acceleration=&RandomRangeVector(Minimum=<0 0 0>, Maximum=<0 0 50>))
   &ApplyInitialForces()
   &ShapeLocation(MaskSpawnGroup=false, Override Seed=false, Sphere Radius=55, UseOwnerScale=true)
   &DLR_ProductPreviews(GameModeID:NiagaraInt32=$System.GameModeID, bPreviewVehicleSpeed:NiagaraBool=$System.bPreviewVehicleSpeed)
  ]
  particleUpdateScript [
   &ParticleState()
   &self:WriteDeath_02(Position Array:NiagaraDataInterfaceArrayPosition=$System.DeathArray) {module=WriteDeath_02}
   &ScaleSpriteSize(Uniform Curve Sprite Scale=DataInterface() {propertyValues {Curve {keys [
        <at 0 0.1 RCIM_Cubic>
        <at 0.1 1>
        {arriveTangent=2.2222, interpMode=RCIM_Cubic, leaveTangent=2.2222, time=0.8, value=1.5}
        <at 1 3 RCIM_Cubic>
       ]}}})
   &ScaleVelocity(
    Coordinate Space:ENiagaraCoordinateSpace=::Simulation
    Velocity Scale=&VectorFromFloat(Value=&FloatFromCurve(FloatCurve=DataInterface() {propertyValues {Curve {keys [<at 0.01 1 RCIM_Cubic>,<at 0.5 0.3 RCIM_Cubic>]}}}))
   )
   &VortexForce(
    UseOriginOffset=false
    Vortex Axis=&RandomRangeVector(Minimum=<0 0 0>, Maximum=<0 16 0>)
    Vortex Force Amount=300
   )
   &InheritSourceMovement(
    Applied Position Scale=&Multiply_VectorByFloat(
     Vector=<1 1 0>
     Float=&FloatFromCurve(FloatCurve=DataInterface() {propertyValues {Curve {keys [
          {
           arriveTangentWeight=0.4252
           interpMode=RCIM_Cubic
           leaveTangentWeight=0.4252
           tangentMode=RCTM_User
           tangentWeightMode=RCTWM_WeightedBoth
           time=0.05
           value=1
          }
          {interpMode=RCIM_Cubic, tangentMode=RCTM_User, time=1, value=0}
         ]}}})
    )
    ApplyForce=false
    ApplyPosition=true
    ApplyPositionFalloff=false
   ) {enabled=false}
   &InheritSourceMovement(
    Applied Position Scale=&Multiply_VectorByFloat(
     Vector=<1 1 1>
     Float=&FloatFromCurve(FloatCurve=DataInterface() {propertyValues {Curve {
         defaultValue=1
         keys [
          {arriveTangent=-1.7885, interpMode=RCIM_Cubic, leaveTangent=-1.7885, tangentMode=RCTM_User, time=0.05, value=1}
          {interpMode=RCIM_Cubic, tangentMode=RCTM_User, time=1, value=0}
         ]
        }}})
    )
    ApplyForce=false
    ApplyPosition=true
    ApplyPositionFalloff=false
   ) {module=InheritSourceMovement001}
   &JitterPosition(Jitter Amount=1.5)
   &SolveForcesAndVelocity(Clamp Velocity=false, Limit Acceleration=false)
  ]
  rendererClasses [&NiagaraSpriteRendererProperties]
  renderers [{rendererClass=&NiagaraSpriteRendererProperties}]
 }
 NE_VC_SplashHitBubbles001 {
  emitterUpdateScript [
   &EmitterState(Life Cycle Mode=::NewEnumerator1, Loop Duration=2, UseLoopDelay=false)
   &SpawnPerFrame(Spawn:NiagaraBool=$System.DeathSpawn, Spawn Count=2, Use Spawn Probability=false)
   &VC_FilteredSocketCount()
   &SpawnRate(SpawnRate=40, Use Spawn Probability=false) {enabled=false}
  ]
  particleSpawnScript [
   &VC_PaintedVariantsOverride() {enabled=false}
   &InitializeParticle(
    Lifetime Max=0.4
    Lifetime Min=0.1
    Lifetime Mode=::NewEnumerator1
    Position=$System.DeathPosition
    Position Mode=::NewEnumerator1
    Sprite Rotation Mode=::NewEnumerator2
    Sprite Size Mode=::NewEnumerator1
    Uniform Sprite Size Max=40
    Uniform Sprite Size Min=20
    UsePositionOffset=false
    Write Material Random=true
   )
   &PartitionParticles(Number Of Partitions=$Emitter.PortCount) {enabled=false}
   &SkeletalMeshLocation(
    Bone / Socket Index=$Particles.PartitionParticles.Partition
    Bone / Socket Sampling Mode=::NewEnumerator5
    MaskSpawnGroup=false
    Mesh Sampling Type=::NewEnumerator6
    Skeletal Mesh=$System.SkeletalMesh
   ) {enabled=false}
   &AddVelocity(Velocity=<-65 0 10>)
   &ApplyInitialForces()
  ]
  particleUpdateScript [
   &ParticleState()
   &ScaleSpriteSize(Uniform Curve Sprite Scale=DataInterface() {propertyValues {Curve {keys [{tangentMode=RCTM_User, time=0, value=0.2}, {tangentMode=RCTM_User, time=1, value=4}]}}})
   &ScaleColor(
    Scale Alpha=&FloatFromCurve(FloatCurve=DataInterface() {propertyValues {Curve {keys [<at 0 0.5 RCIM_Cubic>,<at 1.0011 0.002 RCIM_Cubic>]}}})
    Scale RGB=<2 2 2>
    ScaleA=true
    ScaleRGB=true
   )
   &SolveForcesAndVelocity(Clamp Velocity=false, Limit Acceleration=false)
   &SubUVAnimation(
    "Sprite Renderer "=DataInterface() {propertyValues {SpriteRenderer=&self:NiagaraEmitter_4.NiagaraSpriteRendererProperties_3}}
    UseEndFrame=false
    UseStartFrame=false
   )
  ]
  rendererClasses [&NiagaraSpriteRendererProperties]
  renderers [{rendererClass=&NiagaraSpriteRendererProperties}]
 }
}
summary {
 systemName=NS_Boost_Bubble
 emitters [
  {emitterName=NE_VC_Bubbles, rendererClasses [&NiagaraSpriteRendererProperties]}
  {emitterName=NE_VC_Bubbles001, rendererClasses [&NiagaraSpriteRendererProperties]}
  {emitterName=NE_VC_SmallBubbles, rendererClasses [&NiagaraSpriteRendererProperties]}
  {emitterName=NE_VC_SplashHitBubbles001, rendererClasses [&NiagaraSpriteRendererProperties]}
 ]
 userVariables [
  User.bIsFrontendPreview {type=NiagaraBool}
  User.bIsPainted {type=NiagaraBool}
  User.blsIsCoroColor {type=NiagaraBool}
  User.bPreviewVehicleSpeed {type=NiagaraBool}
  User.ColorVariantID {type=NiagaraInt32}
  User.DTCarOffset {type=NiagaraPosition, description="", defaultValue {}}
  User.GameModeID {type=NiagaraInt32, defaultValue=3}
  User.PaintedColor {type=LinearColor, description="", defaultValue {struct=LinearColor, value=<rgba 1 1 1 1>}}
  User.PaintedScalar {type=NiagaraFloat, defaultValue=1, description=""}
  User.PortCount {type=NiagaraInt32, defaultValue=2}
 ]
}
